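# Translucent, two-sided unlit Niagara sprite material for a dark ink flame effect in the BirdNestBlue emote.
# Gradient-masked noise warps the base UVs, which drive two panning swirl noises that distort a tapered flame shape and a color texture.
# The texture, tinted by Color (black by default) and particle color, feeds emissive; the shape times particle alpha gives opacity, depth-faded outside forward and mobile.
# Dynamic parameter TimeInput can override animation time.
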Material /Game/Effects/Fort_Effects/Emotes/BirdNestBlue/M_BirdNestBlue_DarkInkFlame
graph {
 properties {
  bUsedWithNiagaraSprites=true
  blendMode=BLEND_Translucent
  pixelDepthOffsetMode=PDOM_Legacy
  shadingModel=MSM_Unlit
  twoSided=true
  usedShadingModels=Unlit
 }
 nodes [
  n1 = TextureCoordinate()
  n2 = Add(
   A=OneMinus(None=MaterialFunctionCall(Float2=@n1) {materialFunction=&BreakOutFloat2Components}.G)
   B=ScalarParameter() {group=TotalUVNoise, parameterName=TotalUVNoise_GrayAdjust_Add}
  )
  n3 = ScalarParameter() {defaultValue=1, group=TotalUVNoise, parameterName=TotalUVNoise_GrayAdjust_Mul}
  n4 = Power(
   Base=Saturate(None=Multiply(A=@n2, B=@n3))
   Exp=ScalarParameter() {defaultValue=1, group=TotalUVNoise, parameterName=TotalUVNoise_GrayAdjust_Power}
  )
  n5 = TextureObjectParameter() {
   group=TotalUVNoise
   parameterName=TotalUVNoise_Texture
   samplerType=SAMPLERTYPE_Normal
   sortPriority=0
   texture=&/T_LowResSimplePerlin
  }
  n6 = VectorParameter() {group=TotalUVNoise, parameterName=TotalUVNoise_UVControl, channelNames {a=VPan, b=UPan, g=VScale, r=UScale}, defaultValue=<rgba 1 1 0 0>}
  n7 = MaterialFunctionCall(
   Texture=@n5
   UVControl_RGB=@n6
   UVControl_A=@n6
   "[Override]TimeInput"=DynamicParameter() {defaultValue=<rgba 0 0 1 1>, paramNames [TimeInput,Random,Param3,Param4]}.TimeInput
  ) {materialFunction=&FXCommon_TextureSample}
  n8 = VectorParameter() {
   group=TotalUVNoise
   parameterName=TotalUVNoise_NoiseControl
   sortPriority=60
   channelNames {a=VStrength, b=UStrength, g=VOffset, r=UOffset}
   defaultValue=<rgba -0.5 -0.5 0 0>
  }
  n9 = TextureCoordinate()
  n10 = VectorParameter() {
   group=MainShape
   parameterName=TotalUVOffsetAndLerpScaleU
   channelNames {a="", b="", g=VOffset, r=UOffset}
   defaultValue=<rgba 0 -0.232 0.2 1>
  }
  n11 = AppendVector(A=@n10.UOffset, B=@n10.UOffset)
  n12 = VectorParameter() {group=MainShape, parameterName=TotalUVControl, channelNames {a=VPan, b=UPan, g=VScale, r=UScale}, defaultValue=<rgba 1.3736 1.6353 0 0>}
  n13 = Add(
   A=Multiply(
    A=Saturate(None=@n4)
    B=MaterialFunctionCall!(NoiseValue=@n7.RG, Control_RGB=@n8, Control_A=@n8) {materialFunction=/Game/Effects/Niagara/MasterMaterials/Functions/Common/FXCommon_NoiseControl}.NoiseValue
   )
   B=MaterialFunctionCall!(UVInput=Add(A=@n9, B=@n11), TimeInput=NamedRerouteUsage(), Control_RGB=@n12, Control_A=@n12, ScaleUVByCenter=StaticBool() {value=true}) {materialFunction=/Game/Effects/Niagara/MasterMaterials/Functions/Common/FXCommon_QuickUVControl_V2}.UV
  )
  n14 = MaterialFunctionCall(Float2=Saturate(None=@n13)) {materialFunction=&BreakOutFloat2Components}
  n15 = OneMinus(None=@n14.G)
  n16 = ScalarParameter() {defaultValue=1.2, group=NoiseMask, parameterName=NoiseMask_GrayAdjust_Mul}
  n17 = Power(
   Base=Saturate(None=Multiply(A=@n15, B=@n16))
   Exp=ScalarParameter() {defaultValue=2, group=NoiseMask, parameterName=NoiseMask_GrayAdjust_Power}
  )
  n18 = TextureObjectParameter() {
   group=NoiseA
   parameterName=NoiseA_Texture
   samplerType=SAMPLERTYPE_Normal
   sortPriority=0
   texture=&/T_SwirlNoise
  }
  n19 = VectorParameter() {group=NoiseA, parameterName=NoiseA_UVControl, channelNames {a=VPan, b=UPan, g=VScale, r=UScale}, defaultValue=<rgba 1.7 0.3 0 3>}
  n20 = NamedRerouteUsage()
  n21 = MaterialFunctionCall(Texture=@n18, UVControl_RGB=@n19, UVControl_A=@n19, "[Override]TimeInput"=@n20, "[Override]UVInput"=@n13) {materialFunction=&FXCommon_TextureSample}
  n22 = VectorParameter() {
   group=NoiseA
   parameterName=NoiseA_NoiseControl
   sortPriority=60
   channelNames {a=VStrength, b=UStrength, g=VOffset, r=UOffset}
   defaultValue=<rgba 0 -0.048 0.5 0.7>
  }
  n23 = TextureObjectParameter() {
   group=NoiseB
   parameterName=NoiseB_Texture
   samplerType=SAMPLERTYPE_Normal
   sortPriority=0
   texture=&T_SwirlNoise
  }
  n24 = VectorParameter() {group=NoiseB, parameterName=NoiseB_UVControl, channelNames {a=VPan, b=UPan, g=VScale, r=UScale}, defaultValue=<rgba 1.5 0.2976 0 2>}
  n25 = MaterialFunctionCall(
   Texture=@n23
   AdditionalUVOffset=ScalarParameter() {defaultValue=15, parameterName=NoiseBOffset}
   UVControl_RGB=@n24
   UVControl_A=@n24
   "[Override]TimeInput"=@n20
   "[Override]UVInput"=@n13
  ) {materialFunction=&FXCommon_TextureSample}
  n26 = VectorParameter() {
   group=NoiseB
   parameterName=NoiseB_NoiseControl
   sortPriority=60
   channelNames {a=VStrength, b=UStrength, g=VOffset, r=UOffset}
   defaultValue=<rgba 0 0.08 0.452 0.664>
  }
  n27 = Multiply(
   A=Add(A=Saturate(None=@n17), B=ScalarParameter() {group=NoiseMask, parameterName=NoiseMask_GrayAdjust_Add})
   B=Add(
    A=MaterialFunctionCall(NoiseValue=@n21.RG, Control_RGB=@n22, Control_A=@n22) {materialFunction=&FXCommon_NoiseControl}.NoiseValue
    B=MaterialFunctionCall(NoiseValue=@n25.RG, Control_RGB=@n26, Control_A=@n26) {materialFunction=&FXCommon_NoiseControl}.NoiseValue
   )
  )
  n28 = VectorParameter() {group=Texture, parameterName=Texture_UVControl, channelNames {a=VPan, b=UPan, g=VScale, r=UScale}, defaultValue=<rgba 1 1 0 0>}
  n29 = MaterialFunctionCall(
   Texture=TextureObjectParameter() {group=Texture, parameterName=Texture_Texture, sortPriority=0}
   AdditionalUVOffset=Multiply(A=@n27, B=ScalarParameter() {defaultValue=0.5, parameterName=Texture_NoiseStr})
   UVControl_RGB=@n28
   UVControl_A=@n28
   "[Override]TimeInput"=NamedRerouteUsage()
  ) {materialFunction=&FXCommon_TextureSample}
  n30 = ScalarParameter() {group=Texture, parameterName=Texture_GrayAdjust_Add}
  n31 = Saturate(None=Multiply(A=Add(A=@n29.R, B=@n30), B=ScalarParameter() {defaultValue=1, group=Texture, parameterName=Texture_GrayAdjust_Mul}))
  n32 = ScalarParameter() {defaultValue=1, group=Texture, parameterName=Texture_GrayAdjust_Power}
  n33 = Multiply(A=Saturate(None=Power(Base=@n31, Exp=@n32)), B=VectorParameter() {parameterName=Color, defaultValue=<rgba 0 0 0 1>}.RGB)
  n34 = ParticleColor()
  MaterialFunctionCall(Emissive=Divide(A=Multiply(A=@n33, B=@n34.RGB))) {materialFunction=&MF_FX_EmissiveAdjustments}
  n35 = DynamicParameter() {defaultValue=<rgba 0 0 1 1>, paramNames [TimeInput,Random,Param3,Param4]}
  n36 = Add(A=@n35.TimeInput, B=@n35.TimeInput)
  n37 = Time() {bOverride_Period=true, period=6666}
  NamedRerouteDeclaration(None=If(A=@n36, "A > B"=@n36, "A == B"=@n37, "A < B"=@n37)) {name=DPTime, nodeColor=<rgba 1 0.7266 0 1>}
  n38 = TextureObjectParameter() {group=MainShape, parameterName=MainShape_Texture, sortPriority=0, texture=&/T_BaseFlameShape}
  n39 = VectorParameter() {group=MainShape, parameterName=MainShape_UVControl, channelNames {a=VPan, b=UPan, g=VScale, r=UScale}, defaultValue=<rgba 1 1 0 0>}
  n40 = LinearInterpolate(A=@n10.B, B=@n10.B, Alpha=@n14.G)
  n41 = Constant() {r=1}
  n42 = MaterialFunctionCall!(
   Texture=@n38
   AdditionalUVOffset=@n27
   UVControl_RGB=@n39
   UVControl_A=@n39
   "[Override]TimeInput"=NamedRerouteUsage()
   "[Override]ScaleUVByCenter"
   "[Override]UVInput"=MaterialFunctionCall(
    UVs=@n13
    Texture Scale=MaterialFunctionCall(X=@n40, Y=@n41) {materialFunction=&MakeFloat2}.Result
   ) {materialFunction=&ScaleUVsByCenter}.UVs
  ) {materialFunction=/Game/Effects/Niagara/MasterMaterials/Functions/Common/FXCommon_TextureSample_V2}
  n43 = VectorParameter() {parameterName=TilingMethod, defaultValue=<rgba 1 1 0 1>}
  n44 = AppendVector(A=@n43.R, B=@n43.R)
  n45 = Multiply(
   A=@n34.A
   B=Multiply(
    A=MaterialFunctionCall!(TexCoord=@n42.UV, ClampChannel=@n44) {materialFunction=/Game/Effects/Fort_Effects/MaterialFunctions/MF_FX_TilingMethod}.Result
    B=@n42.R
   )
  )
  n46 = DepthFade(Opacity=@n45, FadeDistance=ScalarParameter() {parameterName=DepthFade}) {fadeDistanceDefault=5}
  ShadingPathSwitch(Default=@n46, Deferred=@n46, Forward=@n45, Mobile=@n45)
 ]
}
